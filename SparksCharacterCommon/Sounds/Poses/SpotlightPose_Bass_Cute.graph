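# Sound cue for the cute bass spotlight pose, routed to the in-game SFX sound class.
# A mixer layers the Pose_Bass_Cute wave with a quieter cheer branch.
# The cheer branch starts after a short fixed delay and picks one of three equally weighted FM pose cheer waves at random.
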SoundCue /SparksCharacterCommon/Sounds/Poses/SpotlightPose_Bass_Cute
root=Mixer_0
nodes [
 {id=Delay_0, children [Random_1], properties {delayMax=0.25, delayMin=0.25}}
 {id=Mixer_0, children [WavePlayer_0,Delay_0], properties {inputVolume [0.5,0.2]}}
 {id=Random_1, children [WavePlayer_6,WavePlayer_5,WavePlayer_4], properties {weights [1,1,1]}}
 {id=WavePlayer_0, properties {soundWaveAssetPtr=&/Pose_Bass_Cute}}
 {id=WavePlayer_4, properties {soundWaveAssetPtr=&/FM_pose_cheer03}}
 {id=WavePlayer_5, properties {soundWaveAssetPtr=&/FM_pose_cheer02}}
 {id=WavePlayer_6, properties {soundWaveAssetPtr=&/FM_pose_cheer01}}
]
properties {maxDistance=2097152, soundClassObject=&Fort_SFX_InGame}
