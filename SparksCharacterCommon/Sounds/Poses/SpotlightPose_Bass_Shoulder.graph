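# Sound cue for the Bass Shoulder spotlight pose in Sparks character audio, routed to the in-game SFX sound class.
# A mixer layers the Pose_Bass_Shoulder wave with a quieter crowd-cheer branch.
# That branch waits a fixed quarter second, then plays one of three equally weighted cheer waves at random for variation.
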SoundCue /SparksCharacterCommon/Sounds/Poses/SpotlightPose_Bass_Shoulder
root=Mixer_0
nodes [
 {id=Delay_0, children [Random_0], properties {delayMax=0.25, delayMin=0.25}}
 {id=Mixer_0, children [WavePlayer_1,Delay_0], properties {inputVolume [0.5,0.2]}}
 {id=Random_0, children [WavePlayer_3,WavePlayer_2,WavePlayer_0], properties {weights [1,1,1]}}
 {id=WavePlayer_0, properties {soundWaveAssetPtr=&/FM_pose_cheer03}}
 {id=WavePlayer_1, properties {soundWaveAssetPtr=&/Pose_Bass_Shoulder}}
 {id=WavePlayer_2, properties {soundWaveAssetPtr=&/FM_pose_cheer02}}
 {id=WavePlayer_3, properties {soundWaveAssetPtr=&/FM_pose_cheer01}}
]
properties {maxDistance=2097152, soundClassObject=&Fort_SFX_InGame}
